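# Physics asset for the SteamPower backpack cosmetic mesh, defining collision bodies and a joint for its wing bones.
# Tiny capsules sit on the left and right base wing bones, while a larger, offset capsule covers the right dynamic wing.
# A single constraint joins the right dynamic wing to its base with tight swing and twist limits.
# The left side has no dynamic wing body or constraint.
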PhysicsAsset /BRCosmetics/Accessories/FORT_Backpacks/Backpack_M_MED_SteamPower/Meshes/M_MED_SteamPower_Pack_Physics
bodies {
 base_wing_l [{shapeType=Capsule, length=0.505, radius=0.505, shapeName=base_wing_l_capsule, rotation=<rot 0 -90 0>}]
 base_wing_r [{shapeType=Capsule, length=0.505, radius=0.505, shapeName=base_wing_r_capsule, rotation=<rot 0 -90 0>}]
 dyn_wing_r [{
   shapeType=Capsule
   length=24.6149
   radius=4.8876
   shapeName=dyn_wing_r_capsule
   center=<14.5501 8.2543 7.5485>
   rotation=<rot -1.7253 121.4711 -52.2073>
  }]
}
constraints [{bone1Name=dyn_wing_r, bone2Name=base_wing_r, swing1LimitDegrees=5, swing2LimitDegrees=5, twistLimitDegrees=1}]
