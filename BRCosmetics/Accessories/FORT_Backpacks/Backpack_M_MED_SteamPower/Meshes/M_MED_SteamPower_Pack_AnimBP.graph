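# Animation Blueprint for the SteamPower backpack cosmetic, adding physics and state-driven wing poses.
# A RigidBody node simulates the pack, using a gravity override the EventGraph refreshes each update.
# Chained cached poses then layer bool-blended additive yaw on both wing bones for sprint, tactical sprint, swim sprint, skydive, parachute, jump and fall.
# A final pack bone offset and rotation applies in the front end.
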AnimBlueprint /BRCosmetics/Accessories/FORT_Backpacks/Backpack_M_MED_SteamPower/Meshes/M_MED_SteamPower_Pack_AnimBP
target_skeleton=&/M_MED_SteamPower_Pack_Skeleton
graphs [
 AnimGraph {nodes [
   n1 = Get bIsJumping()
   n2 = "Transform (Modify) Bone - Bone: base_wing_r"!(ComponentPose=Local To Component(LocalPose=Use cached pose 'Skydive_parachute_Cache'()), bAlphaBoolEnabled:bool=@n1, Rotation="0,10,0") -> (Pose) {
    alphaInputType=Bool
    rotationMode=BMM_Additive
    alphaBoolBlend {blendInTime=0.25, blendOption=Linear, blendOutTime=0.25}
    alphaScaleBiasClamp {clampMax=0.8, interpSpeedIncreasing=15}
    boneToModify {boneName=base_wing_r}
    rotation {yaw=10}
   }
   n3 = "Transform (Modify) Bone - Bone: base_wing_l"!(ComponentPose=@n2, bAlphaBoolEnabled:bool=@n1, Rotation="0,-10,0") -> (Pose) {
    alphaInputType=Bool
    rotationMode=BMM_Additive
    alphaBoolBlend {blendInTime=0.25, blendOption=Linear, blendOutTime=0.25}
    alphaScaleBiasClamp {clampMax=0.8, interpSpeedIncreasing=15}
    boneToModify {boneName=base_wing_l}
    rotation {yaw=-10}
   }
   n4 = Get bIsFalling()
   n5 = "Transform (Modify) Bone - Bone: base_wing_r"(ComponentPose=@n3, bAlphaBoolEnabled=@n4, Rotation="0,15,0") {
    alphaInputType=Bool
    rotationMode=BMM_Additive
    alphaBoolBlend {blendInTime=0.25, blendOption=Linear, blendOutTime=0.25}
    alphaScaleBiasClamp {clampMax=0.8, interpSpeedIncreasing=15}
    boneToModify {boneName=base_wing_r}
    rotation {yaw=15}
   }
   n6 = "Transform (Modify) Bone - Bone: base_wing_l"(ComponentPose=@n5, bAlphaBoolEnabled=@n4, Rotation="0,-15,0") {
    alphaInputType=Bool
    rotationMode=BMM_Additive
    alphaBoolBlend {blendInTime=0.25, blendOption=Linear, blendOutTime=0.25}
    alphaScaleBiasClamp {clampMax=0.8, interpSpeedIncreasing=15}
    boneToModify {boneName=base_wing_l}
    rotation {yaw=-15}
   }
   n7 = "Transform (Modify) Bone - Bone: pack"!(ComponentPose=@n6, bAlphaBoolEnabled:bool=Get bIsInFrontEnd(), Translation="0,-1.5,0", Rotation="0,25,0", Scale) -> (Pose) {
    alphaInputType=Bool
    translationMode=BMM_Additive
    translationSpace=BCS_BoneSpace
    alphaBoolBlend {blendInTime=0.25, blendOption=Linear, blendOutTime=0.25}
    boneToModify {boneName=pack}
    rotation {yaw=25}
    translation {y=-1.5}
   }
   Output Pose(Result=Component To Local(ComponentPose=@n7)) {layerGroup=None}
   Save cached pose 'Input_Cache'(Pose=Input Pose()) {cacheName=Input_Cache}
   n8 = RigidBody(ComponentPose=Local To Component(LocalPose=Use cached pose 'Input_Cache'()), ExternalForce=Get GravityOverride()) {
    overridePhysicsAsset=&/M_MED_SteamPower_Pack_Physics
    alphaBoolBlend {blendOption=Linear}
    alphaScaleBiasClamp {bInterpResult=true}
    baseBoneRef {boneName=root}
    componentLinearAccScale=<0.4 0.3 0.3>
    componentLinearVelScale=<0.4 0.3 0.3>
   }
   n9 = "Transform (Modify) Bone - Bone: pack"!(ComponentPose=@n8, Alpha:float, Translation="0,-0.5,0", Rotation="0,0,0", Scale) -> (Pose) {
    translationMode=BMM_Additive
    translationSpace=BCS_ParentBoneSpace
    alphaBoolBlend {blendOption=Linear}
    boneToModify {boneName=pack}
    translation {y=-0.5}
   }
   Save cached pose 'RigidBody_Cache'(Pose=Component To Local(ComponentPose=@n9)) {cacheName=RigidBody_Cache}
   n10 = Get bIsSprinting()
   n11 = "Transform (Modify) Bone - Bone: base_wing_r"!(
    ComponentPose=Local To Component(LocalPose=Use cached pose 'RigidBody_Cache'())
    bAlphaBoolEnabled=@n10
    Translation:Vector="0,0,0"
    Rotation="0,15,0"
    Scale:Vector="1,1,1"
   ) -> (Pose) {
    alphaInputType=Bool
    rotationMode=BMM_Additive
    alphaBoolBlend {blendInTime=0.25, blendOption=Linear, blendOutTime=0.25}
    boneToModify {boneName=base_wing_r}
    rotation {yaw=15}
   }
   n12 = "Transform (Modify) Bone - Bone: base_wing_l"!(ComponentPose=@n11, bAlphaBoolEnabled=@n10, Translation:Vector="0,0,0", Rotation="0,-15,0", Scale:Vector="1,1,1") -> (Pose) {
    alphaInputType=Bool
    rotationMode=BMM_Additive
    alphaBoolBlend {blendInTime=0.25, blendOption=Linear, blendOutTime=0.25}
    boneToModify {boneName=base_wing_l}
    rotation {yaw=-15}
   }
   n13 = Get bIsTacticalSprinting()
   n14 = "Transform (Modify) Bone - Bone: base_wing_r"(ComponentPose=@n12, bAlphaBoolEnabled=@n13, Translation="0,0,0", Rotation="0,25,0", Scale="1,1,1") {
    alphaInputType=Bool
    rotationMode=BMM_Additive
    alphaBoolBlend {blendInTime=0.25, blendOption=Linear, blendOutTime=0.25}
    boneToModify {boneName=base_wing_r}
    rotation {yaw=25}
   }
   n15 = "Transform (Modify) Bone - Bone: base_wing_l"(ComponentPose=@n14, bAlphaBoolEnabled=@n13, Translation="0,0,0", Rotation="0,-25,0", Scale="1,1,1") {
    alphaInputType=Bool
    rotationMode=BMM_Additive
    alphaBoolBlend {blendInTime=0.25, blendOption=Linear, blendOutTime=0.25}
    boneToModify {boneName=base_wing_l}
    rotation {yaw=-25}
   }
   Save cached pose 'Sprint_Cache'(Pose=Component To Local(ComponentPose=@n15)) {cacheName=Sprint_Cache}
   n16 = Get bIsSwimSprinting() -> (bIsSwimSprinting:bool)
   n17 = "Transform (Modify) Bone - Bone: base_wing_r"!(ComponentPose=Local To Component(LocalPose=Use cached pose 'Sprint_Cache'()), bAlphaBoolEnabled=@n16, Rotation="0,20,0") -> (Pose) {
    alphaInputType=Bool
    rotationMode=BMM_Additive
    alphaBoolBlend {blendInTime=0.25, blendOption=Linear, blendOutTime=0.25}
    boneToModify {boneName=base_wing_r}
    rotation {yaw=20}
   }
   n18 = "Transform (Modify) Bone - Bone: base_wing_l"!(ComponentPose=@n17, bAlphaBoolEnabled=@n16, Rotation="0,-20,0") -> (Pose) {
    alphaInputType=Bool
    rotationMode=BMM_Additive
    alphaBoolBlend {blendInTime=0.25, blendOption=Linear, blendOutTime=0.25}
    boneToModify {boneName=base_wing_l}
    rotation {yaw=-20}
   }
   Save cached pose 'Swim_Cache'(Pose=Component To Local(ComponentPose=@n18)) {cacheName=Swim_Cache}
   n19 = Get bIsSkydiveDiveMode()
   n20 = "Transform (Modify) Bone - Bone: base_wing_r"(ComponentPose=Local To Component(LocalPose=Use cached pose 'Swim_Cache'()), bAlphaBoolEnabled=@n19, Rotation="0,20,0") {
    alphaInputType=Bool
    rotationMode=BMM_Additive
    alphaBoolBlend {blendInTime=0.25, blendOption=Linear, blendOutTime=0.25}
    alphaScaleBiasClamp {clampMax=0.8, interpSpeedIncreasing=15}
    boneToModify {boneName=base_wing_r}
    rotation {yaw=20}
   }
   n21 = "Transform (Modify) Bone - Bone: base_wing_l"(ComponentPose=@n20, bAlphaBoolEnabled=@n19, Rotation="0,-20,0") {
    alphaInputType=Bool
    rotationMode=BMM_Additive
    alphaBoolBlend {blendInTime=0.25, blendOption=Linear, blendOutTime=0.25}
    alphaScaleBiasClamp {clampMax=0.8, interpSpeedIncreasing=15}
    boneToModify {boneName=base_wing_l}
    rotation {yaw=-20}
   }
   n22 = Get bIsParachuteOpened()
   n23 = "Transform (Modify) Bone - Bone: base_wing_r"(ComponentPose=@n21, bAlphaBoolEnabled=@n22, Rotation="0,25,0") {
    alphaInputType=Bool
    rotationMode=BMM_Additive
    alphaBoolBlend {blendInTime=0.25, blendOption=Linear, blendOutTime=0.25}
    alphaScaleBiasClamp {clampMax=0.8, interpSpeedIncreasing=15}
    boneToModify {boneName=base_wing_r}
    rotation {yaw=25}
   }
   n24 = "Transform (Modify) Bone - Bone: base_wing_l"(ComponentPose=@n23, bAlphaBoolEnabled=@n22, Rotation="0,-25,0") {
    alphaInputType=Bool
    rotationMode=BMM_Additive
    alphaBoolBlend {blendInTime=0.25, blendOption=Linear, blendOutTime=0.25}
    alphaScaleBiasClamp {clampMax=0.8, interpSpeedIncreasing=15}
    boneToModify {boneName=base_wing_l}
    rotation {yaw=-25}
   }
   Save cached pose 'Skydive_parachute_Cache'(Pose=Component To Local(ComponentPose=@n24)) {cacheName=Skydive_parachute_Cache}
  ]}
 EventGraph {nodes [
   Event BlueprintUpdateAnimation()
   > n1 = Gravity Override State Function DEPRECATED(CharacterPartAnimInstance=Self-Reference(), GameTimeInSeconds=GetGameTimeInSeconds(), GravityOverrideParams=Get Gravity Override Params())
   > Set GravityOverride(GravityOverride=@n1)
   TryGetPawnOwner()
  ]}
]
variables [
 Gravity Override Params {type=GravityOverrideParamsStruct}
 GravityOverride {type=Vector}
]
